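# Physics asset for the Ms Whip backpack cosmetic mesh, giving its whip part a limited physical range of motion.
# It defines small capsule bodies on the pack bone and the dyn_whip bone.
# A single constraint joins dyn_whip to pack with narrow swing limits and an even tighter twist limit, so the whip can only sway slightly relative to the pack.
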PhysicsAsset /BRCosmetics/Accessories/FORT_Backpacks/Backpack_F_MED_Ms_Whip/Meshes/F_MED_Ms_Whip_Pack_Dyn
bodies {
 dyn_whip [{shapeType=Capsule, length=2, radius=2, center=<0.5513 -0.0001 0>, rotation=<rot 90 0 0>}]
 pack [{shapeType=Capsule, length=1, radius=1}]
}
constraints [{bone1Name=dyn_whip, bone2Name=pack, swing1LimitDegrees=20, swing2LimitDegrees=20, twistLimitDegrees=5}]
